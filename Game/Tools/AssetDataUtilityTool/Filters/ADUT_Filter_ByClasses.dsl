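; Asset filter for the AssetDataUtilityTool that passes an asset only if it matches the classes in ReturnOnlyClasses.
; AssetDataPassesFilter converts those soft classes to class paths, then queries the Asset Registry by the asset's package name and those paths, subclasses included.
; The asset passes when the query returns any result.
; ConvertSoftClassToAssetPath accepts only native /Script/ classes, so other classes, such as Blueprint ones, are left out of the query.
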(fn AssetDataPassesFilter
  (params (AssetData :type Asset%20Data%20Structure%20%28by%20ref%29))
  (Utilities|FlowControl|Sequence
    (:then_0
      (for _array_element (Variables|Default|GetReturnOnlyClasses)
        (if (.valid (Default|ConvertSoftClasstoAssetPath _array_element))
          (Utilities|Array|Add (LocalVariables|Getl_classes) (Default|ConvertSoftClasstoAssetPath _array_element)))))
    (:then_1
      (bind _outassetdata (AssetRegistry|GetAssets (AssetRegistry|GetAssetRegistry) (Utilities|MakeARFilter :PackageNames (Utilities|Array|MakeArray (Utilities|Struct|BreakAssetData AssetData)) :ClassPaths (LocalVariables|Getl_classes) :bRecursivePaths false :bRecursiveClasses true :bIncludeOnlyOnDiskAssets false) false))
      (return (Utilities|Array|IsValidIndex _outassetdata)))))

(fn ConvertSoftClassToAssetPath
  (params (SoftClassReference :type Object%20Soft%20Class%20Reference))
  (bind _returnvalue (Utilities|ToSoftClassPath SoftClassReference))
  (bind _pathstring (Utilities|BreakSoftClassPath _returnvalue))
  (bind (_lefts _rights) (Utilities|String|Split _pathstring "."))
  (if (Utilities|String|StartsWith _pathstring "/Script/")
    (return (Utilities|MakeTopLevelAssetPath (Utilities|String|ToString(Name) (Utilities|String|StringToName _lefts)) (Utilities|String|ToString(Name) (Utilities|String|StringToName _rights))) true)
    (else
      (return 0 false))))
